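# Niagara idle effect for the VoyagerCuboid guitar cosmetic: a glowing cube with orange-gold lightning and small drifting cubes.
# MainCube samples a skeletal mesh for position and adds a mesh and warm light.
# Two lightning emitters spawn short-lived meshes probabilistically, animating material parameters and scaling brightness by time of day and front end versus gameplay.
# SmallCubesLinger spawns slowly rotating cubes on an offset ring; SmallCubesOut is disabled.
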NiagaraSystem /SparksCosmetics/Instrument/Guitar/VoyagerCuboid/Effects/NS_Guitar_VoyagerCuboid_Idle
dependencies {
 usedDataInterfaces [&NiagaraDataInterfaceCurve, &NiagaraDataInterfaceColorCurve, &NiagaraDataInterfaceSkeletalMesh]
 usedDynamicInputs [&MakeFloatFromVector]
 usedModules [
  &self:SystemSpawnScript.SystemScriptSource.SystemScriptGraph.NiagaraNodeAssignment_0.SetVariables_#1
  &self:SystemSpawnScript.SystemScriptSource.SystemScriptGraph.NiagaraNodeAssignment_1.SetVariables_#2
  &self:SystemSpawnScript.SystemScriptSource.SystemScriptGraph.NiagaraNodeAssignment_2.SetVariables_#3
  &SystemState
 ]
}
emitters {
 Hunt_Lightning {
  emitterUpdateScript [
   &EmitterState()
   &SpawnRate(
    Spawn Probability=/Niagara/DynamicInputs/Helpers/DistributionCurveFloat(
     Minimum:NiagaraFloat=0.25
     Maximum:NiagaraFloat=0.8
     Distribution Curve:NiagaraDataInterfaceCurve=DataInterface() {propertyValues {Curve {
        defaultValue=0
        keys [{interpMode=RCIM_Cubic, tangentMode=RCTM_User, time=0.5, value=0}, {arriveTangent=-0.0819, tangentMode=RCTM_User, time=1, value=1}]
       }}}
     Randomness Mode:ENiagaraRandomnessMode=::NewEnumerator0
     Evaluation Type:ENiagaraRandomnessEvaluation=::NewEnumerator1
     Seed:NiagaraInt32=0
     Fixed Random Seed:NiagaraBool=false
    )
    SpawnRate=11
    Use Spawn Probability=true
   )
   &SpawnRate(
    Spawn Probability=&DistributionCurveFloat(
     Minimum:NiagaraFloat=0.25
     Maximum:NiagaraFloat=0.8
     Distribution Curve:NiagaraDataInterfaceCurve=DataInterface() {propertyValues {Curve {
        defaultValue=0
        keys [{interpMode=RCIM_Cubic, tangentMode=RCTM_User, time=0.5, value=0}, {arriveTangent=-0.0819, tangentMode=RCTM_User, time=1, value=1}]
       }}}
     Randomness Mode:ENiagaraRandomnessMode=::NewEnumerator0
     Evaluation Type:ENiagaraRandomnessEvaluation=::NewEnumerator1
     Seed:NiagaraInt32=0
     Fixed Random Seed:NiagaraBool=false
    )
    SpawnGroup=1
    SpawnRate=11
    Use Spawn Probability=true
   ) {module=SpawnRate001}
  ]
  particleSpawnScript [
   &InitializeParticle(
    Color=<rgba 4 2.2701 0 1>
    Lifetime Max=0.55
    Lifetime Min=0.2
    Lifetime Mode=::NewEnumerator1
    Mesh Scale=&Multiply_VectorByFloat(
     Vector=&MergeHelper_VectorSwitch(
      0=&Lerp_Vector(
       A=&RandomRangeVector(Minimum=<0.75 0.4 0.12>, Maximum=<1 0.6 0.18>)
       B=&RandomRangeVector(Minimum=<1.25 1 0.3>, Maximum=<1.5 1 0.3>)
       Alpha=&FloatFromCurve(
        FloatCurve=DataInterface() {propertyValues {Curve {keys [{interpMode=RCIM_Cubic, tangentMode=RCTM_User, time=0.5, value=0}, <at 1 1>]}}}
        CurveIndex=&RandomRangeFloat(Minimum=0, Maximum=1)
       )
      )
      1=&Lerp_Vector(
       A=&RandomRangeVector(Minimum=<0.5 0.4 0.12>, Maximum=<0.5 0.6 0.18>)
       B=&RandomRangeVector(Minimum=<0.75 1 0.3>, Maximum=<0.75 1 0.3>)
       Alpha=&FloatFromCurve(
        FloatCurve=DataInterface() {propertyValues {Curve {keys [{interpMode=RCIM_Cubic, tangentMode=RCTM_User, time=0.5, value=0}, {arriveTangent=3.463, tangentMode=RCTM_User, time=1, value=1}]}}}
        CurveIndex=&RandomRangeFloat(Minimum=0, Maximum=1)
       )
      )
      2=?
      3=?
      Index=&SpawnGroup()
      Num=::NewEnumerator0
     )
     Float=&Multiply_Float(A=$System.GlobalScale, B=0.2)
    )
    Mesh Scale Mode=::NewEnumerator4
    UsePositionOffset=true
    Write Material Random=true
   )
   &ShapeLocation(
    Box Size=<0.5 0.5 0.5>
    MaskSpawnGroup=false
    Offset=&MaskVectorBySpawnGroup(Masked Vector=<0 0 0>, Passthrough Vector=<0 0 58>)
    Offset Mode=::NewEnumerator0
    Override Seed=false
    Shape Primitive=::NewEnumerator2
    Surface Only=false
    UseOwnerScale=true
   )
   &InitialMeshOrientation(
    Mesh Axis To Orient=<0 0 1>
    Mesh Orientation Mode=::NewEnumerator2
    Rotation=&MaskVectorBySpawnGroup(
     Masked Vector=&RandomRangeVector(Minimum=<0 0 0>, Maximum=<0.1 0 1>)
     Passthrough Vector=&RandomRangeVector(Minimum=<-0.2 0 -0.2>, Maximum=<-0.2 0 0.2>)
    )
    Use Rotation Vector=true
   )
   &InitialMeshOrientation(Rotation=&RandomRangeVector(Minimum=<-0.5 -0.5 -0.5>, Maximum=<0.5 0.5 0.5>), Use Rotation Vector=true) {enabled=false, module=InitialMeshOrientation002}
   /SetVariables(Particles.Random=&RandomRangeFloat(Minimum=0, Maximum=1)) {isSetParameters=true, module=SetVariables_#4}
  ]
  particleUpdateScript [
   &ParticleState()
   &ScaleColor(
    Scale Alpha=&Multiply_Float(
     A=10
     B=&FloatFromCurve(FloatCurve=DataInterface() {propertyValues {Curve {keys [
          {interpMode=RCIM_Cubic, tangentMode=RCTM_User, time=0, value=0}
          {interpMode=RCIM_Cubic, tangentMode=RCTM_User, time=0.1, value=0}
          {interpMode=RCIM_Cubic, tangentMode=RCTM_User, time=0.2, value=0.8}
          {interpMode=RCIM_Cubic, tangentMode=RCTM_User, time=1, value=0}
         ]}}})
    )
    Scale RGB=&Multiply_VectorByFloat(
     Vector=&Multiply_VectorByFloat(Vector=<1 1 1>, Float=&FN_FrontEndFloatModifier(Gameplay=1, Front End=0.5))
     Float=&Multiply_Float(
      A=&RandomRangeFloat(Minimum=0.5, Maximum=1, Evaluation Type=::NewEnumerator1)
      B=&Multiply_Float(A=&FNTimeOfDayScalar(Night Scale=0.2, Morning Scale=1, Evening Scale=1), B=10)
     )
    )
    ScaleA=true
    ScaleRGB=true
   )
   &DynamicMaterialParameters(
    Index 0 Param 1=&FloatFromCurve(FloatCurve=DataInterface() {propertyValues {Curve {keys [{tangentMode=RCTM_User, time=0, value=0}, {tangentMode=RCTM_User, time=1, value=1}]}}})
    Index 0 Param 3=&FloatFromCurve(FloatCurve=DataInterface() {propertyValues {Curve {keys [
         {arriveTangent=4.0072, interpMode=RCIM_Cubic, leaveTangent=4.0072, tangentMode=RCTM_User, time=0, value=0}
         {arriveTangent=2.7083, interpMode=RCIM_Cubic, leaveTangent=2.7083, time=0.2543, value=0.6225}
         {arriveTangent=1.8473, interpMode=RCIM_Cubic, leaveTangent=1.8473, time=0.6, value=1.625}
         {arriveTangent=1.6993, interpMode=RCIM_Cubic, leaveTangent=1.6993, tangentMode=RCTM_User, time=1, value=2}
        ]}}})
    Index 0 Param 4=20
    Index 1 Param 1=0.4
    Index 1 Param 2=$Particles.Random
    Index 1 Param 3=1.3
    Index 1 Param 4=1
    Index 2 Param 1=&FloatFromCurve(FloatCurve=DataInterface() {propertyValues {Curve {keys [<at 0 1>,<at 1 0.3>]}}})
    Module.Index0Float0Write=true
    Module.Index0Float1Write=false
    Module.Index0Float2Write=true
    Module.Index0Float3Write=true
    Module.Index1Float0Write=true
    Module.Index1Float1Write=true
    Module.Index1Float2Write=true
    Module.Index1Float3Write=true
    Module.Index2Float0Write=true
    Module.Index2Float1Write=false
    Module.Index2Float2Write=false
    Module.Index2Float3Write=false
    Write Parameter Index 1=::NewEnumerator1
    Write Parameter Index 2=::NewEnumerator1
   )
   &SolveForcesAndVelocity(Clamp Velocity=false, Limit Acceleration=false)
   /SetVariables(Particles.Lightning glow color:LinearColor=&MakeLinearColorFromVectorAndFloat(
     "Vector (RGB)"=&Multiply_VectorByFloat(
      Vector=&MakeVectorFromLinearColorRGB(Color=<rgba 1 0.4839 0 1>)
      Float=&FloatFromCurve(FloatCurve=DataInterface() {propertyValues {Curve {keys [
           {arriveTangent=-45.7279, interpMode=RCIM_Cubic, leaveTangent=-45.7279, tangentMode=RCTM_User, time=0, value=5}
           {arriveTangent=-8.2888, interpMode=RCIM_Cubic, leaveTangent=-8.2888, tangentMode=RCTM_User, time=0.1, value=2}
           {arriveTangent=-1.0634, tangentMode=RCTM_User, time=0.5, value=1}
          ]}}})
     )
     "Float (Alpha)"=&Multiply_Float(
      A=&FloatFromCurve(FloatCurve=DataInterface() {propertyValues {Curve {keys [<at 0 1 RCIM_Cubic>, {interpMode=RCIM_Cubic, leaveTangent=-3.6307, time=0.9, value=0}]}}})
      B=5
     )
    )) {isSetParameters=true, module=SetVariables_#5}
   /SetVariables(Particles.Glow Dy02:Vector4f=&MakeVector4(
     X=&FloatFromCurve(FloatCurve=DataInterface() {propertyValues {Curve {keys [{tangentMode=RCTM_User, time=0, value=0.2}, {tangentMode=RCTM_User, time=0.5, value=1}]}}})
     Y=0.7
     Z=0
     W=20
    )) {isSetParameters=true, module=SetVariables_#6}
   /SetVariables(Particles.Glow Dy03:Vector4f=&MakeVector4(
     X=&FloatFromCurve(
      FloatCurve=DataInterface() {propertyValues {Curve {keys [<at 0 1>,<at 1 0.5>]}}}
      Scale Curve=&MaskFloat(Mask Source=1, Mask Parameter=$System.VariantIndex, Mask=2, Masked Float=0.7, Passthrough Float=0.5)
     )
     Y=$Particles.Random
     Z=1.3
     W=1
    )) {isSetParameters=true, module=SetVariables_#7}
  ]
  rendererClasses [&NiagaraMeshRendererProperties]
  renderers [
   {rendererClass=&NiagaraMeshRendererProperties, rendererIndex=0}
   {rendererClass=&NiagaraMeshRendererProperties, rendererIndex=1}
  ]
 }
 LightningStrings {
  emitterUpdateScript [
   &EmitterState()
   &SpawnRate(
    Spawn Probability=&DistributionCurveFloat(
     Minimum:NiagaraFloat=0.3
     Maximum:NiagaraFloat=0.8
     Distribution Curve:NiagaraDataInterfaceCurve=DataInterface() {propertyValues {Curve {
        defaultValue=0
        keys [{interpMode=RCIM_Cubic, tangentMode=RCTM_User, time=0.5, value=0}, {arriveTangent=-0.0819, tangentMode=RCTM_User, time=1, value=1}]
       }}}
     Randomness Mode:ENiagaraRandomnessMode=::NewEnumerator0
     Evaluation Type:ENiagaraRandomnessEvaluation=::NewEnumerator1
     Seed:NiagaraInt32=0
     Fixed Random Seed:NiagaraBool=false
    )
    SpawnRate=10
    Use Spawn Probability=true
   )
   &SpawnBurst_Instantaneous(Spawn Group=1, Spawn Probability=0.9, Use Spawn Probability=true, UseLoopCountLimit=false)
  ]
  particleSpawnScript [
   &InitializeParticle(
    Color Maximum=<rgba 0.5 0.2838 0 1>
    Color Minimum=<rgba 0.1 0.0355 0 1>
    Color Mode=::NewEnumerator2
    Lifetime=&MergeHelper_FloatSwitch(
     0=&RandomRangeFloat(Minimum=0.25, Maximum=1)
     1=0.75
     2=?
     3=?
     Index=&SpawnGroup()
     Num=::NewEnumerator0
    )
    Mesh Scale=&Multiply_VectorByFloat(
     Vector=&Multiply_Vector(
      A=<3 1 0.3>
      B=&Lerp_Vector(
       A=&RandomRangeVector(Minimum=<0.5 0.4 0.4>, Maximum=<0.75 0.6 0.6>)
       B=<1 1 1>
       Alpha=&MergeHelper_FloatSwitch(
        0=&FloatFromCurve(
         FloatCurve=DataInterface() {propertyValues {Curve {keys [{interpMode=RCIM_Cubic, tangentMode=RCTM_User, time=0.5, value=0}, <at 1 1>]}}}
         CurveIndex=&RandomRangeFloat(Minimum=0, Maximum=1)
        )
        1=&RandomRangeFloat(Minimum=0.75, Maximum=1)
        2=?
        3=?
        Index=&SpawnGroup()
        Num=::NewEnumerator0
       )
      )
     )
     Float=0.2
    )
    Mesh Scale Mode=::NewEnumerator4
    UsePositionOffset=true
    Write Material Random=true
   )
   /SetVariables(Particles.WidthPos:NiagaraFloat=&RandomRangeFloat(Minimum=0, Maximum=1)) {isSetParameters=true}
   /SetVariables(Particles.Position=&Add_Vector(
     A=&Add_Vector(
      A=&Lerp_Vector(A=<0 -3 0>, B=<0 3 0>, Alpha=$Particles.WidthPos)
      B=&RandomRangeVector(Minimum=<0 0 -10>, Maximum=<0 0 0>)
     )
     B=<0 0.75 0>
    )) {isSetParameters=true}
   &ShapeLocation(
    Box Size=<0 5 2>
    MaskSpawnGroup=false
    Offset=<0 0 -5>
    Offset Mode=::NewEnumerator0
    Override Seed=false
    Shape Primitive=::NewEnumerator2
    Surface Only=false
    UseOwnerScale=true
   ) {enabled=false}
   &InitialMeshOrientation(
    Mesh Orientation Mode=::NewEnumerator2
    Rotation=&RandomRangeVector(Minimum=<-1 0 0>, Maximum=<1 0 0>)
    Use Rotation Vector=true
   ) {module=InitialMeshOrientation001}
   &InitialMeshOrientation(
    Mesh Axis To Orient=<0 0 1>
    Mesh Orientation Mode=::NewEnumerator2
    Rotation=&Lerp_Vector(A=<0 0 -0.005>, B=<0 0 0.005>, Alpha=$Particles.WidthPos)
    Use Rotation Vector=true
   )
   &InitialMeshOrientation(Rotation=&RandomRangeVector(Minimum=<-0.5 -0.5 -0.5>, Maximum=<0.5 0.5 0.5>), Use Rotation Vector=true) {enabled=false, module=InitialMeshOrientation002}
   &CameraOffset(Camera Offset Amount=2)
   /SetVariables(Particles.Random=&RandomRangeFloat(Minimum=0, Maximum=1)) {isSetParameters=true, module=SetVariables_#4}
  ]
  particleUpdateScript [
   &ParticleState()
   &ScaleMeshSize(Scale Factor=<3 1 0.3>) {enabled=false}
   &ScaleColor(
    Scale Alpha=&Multiply_Float(
     A=10
     B=&FloatFromCurve(FloatCurve=DataInterface() {propertyValues {Curve {keys [
          {interpMode=RCIM_Cubic, tangentMode=RCTM_User, time=0, value=0}
          {interpMode=RCIM_Cubic, tangentMode=RCTM_User, time=0.2, value=0.8}
          {interpMode=RCIM_Cubic, tangentMode=RCTM_User, time=1, value=0}
         ]}}})
    )
    Scale RGB=&Multiply_VectorByFloat(
     Vector=&Multiply_VectorByFloat(Vector=<1 1 1>, Float=&FN_FrontEndFloatModifier(Gameplay=1, Front End=0.5))
     Float=&Multiply_Float(
      A=&RandomRangeFloat(Minimum=0.5, Maximum=1, Evaluation Type=::NewEnumerator1)
      B=&Multiply_Float(A=&FNTimeOfDayScalar(Night Scale=0.2, Morning Scale=1, Evening Scale=1), B=10)
     )
    )
    ScaleA=true
    ScaleRGB=true
   )
   &DynamicMaterialParameters(
    Index 0 Param 1=&FloatFromCurve(FloatCurve=DataInterface() {propertyValues {Curve {keys [
         {arriveTangent=5.1865, interpMode=RCIM_Cubic, leaveTangent=5.1865, tangentMode=RCTM_User, time=0, value=0}
         {tangentMode=RCTM_User, time=0.25, value=1}
        ]}}})
    Index 0 Param 3=&FloatFromCurve(FloatCurve=DataInterface() {propertyValues {Curve {keys [
         {arriveTangent=0.1919, interpMode=RCIM_Cubic, leaveTangent=0.1919, tangentMode=RCTM_User, time=0, value=0}
         {arriveTangent=0.4037, interpMode=RCIM_Cubic, leaveTangent=0.4037, tangentMode=RCTM_User, time=0.8, value=1.625}
         {arriveTangent=3.3027, interpMode=RCIM_Cubic, leaveTangent=3.3027, tangentMode=RCTM_User, time=1, value=2}
        ]}}})
    Index 0 Param 4=3
    Index 1 Param 1=0.4
    Index 1 Param 2=$Particles.Random
    Index 1 Param 3=1.3
    Index 1 Param 4=1
    Index 2 Param 1=&FloatFromCurve(FloatCurve=DataInterface() {propertyValues {Curve {keys [<at 0 1>,<at 1 0.3>]}}})
    Module.Index0Float0Write=true
    Module.Index0Float1Write=false
    Module.Index0Float2Write=true
    Module.Index0Float3Write=true
    Module.Index1Float0Write=true
    Module.Index1Float1Write=true
    Module.Index1Float2Write=true
    Module.Index1Float3Write=true
    Module.Index2Float0Write=true
    Module.Index2Float1Write=true
    Module.Index2Float2Write=true
    Module.Index2Float3Write=true
    Write Parameter Index 1=::NewEnumerator1
    Write Parameter Index 2=::NewEnumerator1
   )
   &SolveForcesAndVelocity(Clamp Velocity=false, Limit Acceleration=false)
   /SetVariables(Particles.Lightning glow color:LinearColor=&MakeLinearColorFromVectorAndFloat(
     "Vector (RGB)"=&MakeVectorFromLinearColorRGB(Color=&MultiplyLinearColor(
       Linear Color=&RandomRangeLinearColor(Minimum=<rgba 2 0.9678 0 1>, Maximum=<rgba 3 1.4517 0 1>)
       Scale Factor=&MakeVector4_FromFloat(Float=&RandomRangeFloat(Minimum=0.5, Maximum=1, Evaluation Type=::NewEnumerator1))
      ))
     "Float (Alpha)"=&Multiply_Float(
      A=&FloatFromCurve(FloatCurve=DataInterface() {propertyValues {Curve {keys [<at 0 1 RCIM_Cubic>, {interpMode=RCIM_Cubic, leaveTangent=-3.6307, time=0.9, value=0}]}}})
      B=5
     )
    )) {isSetParameters=true, module=SetVariables_#5}
   /SetVariables(Particles.Glow Dy02:Vector4f=&MakeVector4(
     X=&FloatFromCurve(FloatCurve=DataInterface() {propertyValues {Curve {keys [{tangentMode=RCTM_User, time=0, value=0.2}, {tangentMode=RCTM_User, time=0.5, value=1}]}}})
     Y=0.7
     Z=0
     W=5
    )) {isSetParameters=true, module=SetVariables_#6}
   /SetVariables(Particles.Glow Dy03:Vector4f=&MakeVector4(
     X=&FloatFromCurve(
      FloatCurve=DataInterface() {propertyValues {Curve {keys [<at 0 1>,<at 1 0.5>]}}}
      Scale Curve=&MaskFloat(Mask Source=1, Mask Parameter=$System.VariantIndex, Mask=2, Masked Float=0.7, Passthrough Float=0.5)
     )
     Y=$Particles.Random
     Z=1.3
     W=1
    )) {isSetParameters=true, module=SetVariables_#7}
  ]
  rendererClasses [&NiagaraMeshRendererProperties]
  renderers [
   {rendererClass=&NiagaraMeshRendererProperties, rendererIndex=0}
   {rendererClass=&NiagaraMeshRendererProperties, rendererIndex=1}
  ]
 }
 MainCube {
  emitterUpdateScript [
   &EmitterState(Life Cycle Mode=::NewEnumerator1, Loop Delay=$System.Delay, Loop Duration=10, UseLoopDelay=true)
   &SpawnBurst_Instantaneous(Use Spawn Probability=false, UseLoopCountLimit=true)
  ]
  particleSpawnScript [
   /SetVariables(Particles.OutlineVis:NiagaraInt32=&MakeIntFromFloat(Float=$User.Variant, Conversion Mode=::NewEnumerator1)) {enabled=false, isSetParameters=true}
   &PartitionParticles(Number Of Partitions=16, Particles Per Partition=1, Partition Mode:ENiagara_PartitionMode_v2=::NewEnumerator1)
   &InitializeParticle(
    Color=<rgba 0.3229 0.1646 0.0165 1>
    Lifetime=2
    Mesh Scale Mode=::NewEnumerator3
    Mesh Uniform Scale=0.46
    UsePositionOffset=false
    Write Material Random=true
   )
  ]
  particleUpdateScript [
   &ParticleState(Kill Particles When Lifetime Has Elapsed=false)
   /SetVariables(Particles.OutlineColor:LinearColor=&Multiply_LinearColorByFloat(
     Linear Color=<rgba 4 4 4 1>
     Float=&FloatFromCurve(
      FloatCurve=DataInterface() {propertyValues {Curve {
         defaultValue=1
         keys [
          {interpMode=RCIM_Cubic, tangentMode=RCTM_User, time=0, value=1}
          {interpMode=RCIM_Cubic, tangentMode=RCTM_User, time=0.5, value=2}
          {tangentMode=RCTM_User, time=1, value=1}
         ]
        }}}
      Scale Curve=&FloatFromCurve(
       FloatCurve=DataInterface() {propertyValues {Curve {keys [{interpMode=RCIM_Cubic, tangentMode=RCTM_User, time=0, value=0}, <at 1 1>]}}}
       CurveIndex=$Emitter.Age
      )
     )
    )) {isSetParameters=true}
   &SkeletalMeshLocation(Bone Sampling Mode=::NewEnumerator4, MaskSpawnGroup=false, Skeletal Mesh=$System.Skeletal Mesh Reader)
   &ScaleMeshSize(Scale Factor=&Multiply_VectorByFloat(
     Vector=<1 1 1>
     Float=&FloatFromCurve(
      FloatCurve=DataInterface() {propertyValues {Curve {keys [{arriveTangent=0.1632, interpMode=RCIM_Cubic, leaveTangent=0.1632, tangentMode=RCTM_User, time=0, value=0.87}, <at 1 0.97>]}}}
      Scale Curve=$System.GlobalScale
     )
    )) {module=ScaleMeshSize001}
   &Light_Attributes(
    Color=1
    Color Value To Scale:LinearColor=/Niagara/DynamicInputs/LinearColor/MakeLinearColorFromVector4(Vector:Vector4f=<1 1 1 1>)
    Falloff Exponent=1
    Light Exponent=2
    Light Radius=20
    Position=1
    Position Offset Coordinate Space:ENiagaraCoordinateSpace=::Local
    Radius=1
    Scale Alpha=40
    Scale RGB:Vector3f=&MakeVectorFromLinearColorRGB(Color=<rgba 1 0.3713 0 1>)
    ScaleA=true
    ScaleRGB=true
    UsePositionOffset=true
   )
  ]
  rendererClasses [&NiagaraMeshRendererProperties, &NiagaraLightRendererProperties]
  renderers [
   {rendererClass=&NiagaraMeshRendererProperties, rendererIndex=0}
   {rendererClass=&NiagaraLightRendererProperties, rendererIndex=1}
  ]
 }
 SmallCubesLinger {
  emitterUpdateScript [
   &EmitterState()
   &SpawnRate(
    SpawnRate=&RandomRangeFloat(Minimum=0.5, Maximum=5, Evaluation Type=::NewEnumerator1)
    Use Spawn Probability=false
   )
  ]
  particleSpawnScript [
   &InitializeParticle(
    Color=<rgba 2 2 2 1>
    Lifetime Max=6
    Lifetime Min=4
    Lifetime Mode=::NewEnumerator1
    Mesh Scale Mode=::NewEnumerator1
    Mesh Uniform Scale Max=0.01
    Mesh Uniform Scale Min=0.008
    UsePositionOffset=false
    Write Material Random=true
   )
   &InitialMeshOrientation(Use Rotation Vector=false)
   &ShapeLocation(
    Disc Coverage=0.85
    MaskSpawnGroup=false
    Non Uniform Scale=<1 0.75 1>
    Offset=&Multiply_VectorByFloat(
     Vector=&Add_Vector(
      A=<2 0 -1>
      B=&Multiply_VectorByFloat(
       Vector=&RandomRangeVector(Minimum=<2.5 0 0>, Maximum=<4 0 0>)
       Float=&UniformAOrBFloat(A=-1, Evaluation Type=::NewEnumerator0, Recalculate Random Each Loop=true, Override Seed=false)
      )
     )
     Float=$System.GlobalScale
    )
    Offset Mode=::NewEnumerator0
    Override Seed=false
    Ring Radius=15
    Rotation Mode=::NewEnumerator1
    Shape Primitive=::NewEnumerator4
    UseOwnerScale=true
   )
   &AddVelocity(
    Module.Use Falloff Distance=false
    Override Seed=false
    Velocity Mode=::NewEnumerator1
    Velocity Speed=&RandomRangeFloat(Minimum=9, Maximum=12)
   ) {enabled=false}
   &CameraOffset(Camera Offset Amount=1)
  ]
  particleUpdateScript [
   &ParticleState()
   &ScaleMeshSize(Scale Factor=&VectorFromFloat(Value=&FloatFromCurve(FloatCurve=DataInterface() {propertyValues {Curve {
         defaultValue=1
         keys [
          {arriveTangent=6.1928, interpMode=RCIM_Cubic, leaveTangent=6.1928, tangentMode=RCTM_User, time=0, value=0}
          {arriveTangent=0.4171, interpMode=RCIM_Cubic, leaveTangent=0.4171, tangentMode=RCTM_User, time=0.15, value=0.9024}
          {arriveTangentWeight=0.2054, interpMode=RCIM_Cubic, leaveTangentWeight=0.0936, tangentMode=RCTM_User, time=0.8, value=1}
          {arriveTangent=-3.0036, interpMode=RCIM_Cubic, leaveTangent=-3.0036, tangentMode=RCTM_User, time=1, value=0}
         ]
        }}})))
   &UpdateMeshOrientation(Rotation Rate=&RandomRangeFloat(Minimum=0.1, Maximum=0.1), Rotation Vector=<1 1 1>)
   &Drag(
    Drag=&FloatFromCurve(
     FloatCurve=DataInterface() {propertyValues {Curve {
        defaultValue=1
        keys [
         {interpMode=RCIM_Cubic, tangentMode=RCTM_User, time=0, value=0}
         {arriveTangent=5.6533, interpMode=RCIM_Cubic, leaveTangent=5.6533, tangentMode=RCTM_User, time=0.223, value=1}
        ]
       }}}
     Scale Curve=&RandomRangeFloat(Minimum=0.5, Maximum=1)
    )
    Use Linear Drag=true
    Use Rotational Drag=true
   ) {enabled=false}
   &SolveForcesAndVelocity(Clamp Velocity=false, Limit Acceleration=false)
   /SetVariables(Particles.PositionMod:NiagaraPosition=&Multiply_VectorByFloat(
     Vector=Linked() {linkedVariable {name=Particles.Position, type=NiagaraPosition}}
     Float=&FloatFromCurve(
      FloatCurve=DataInterface() {propertyValues {Curve {
         defaultValue=1
         keys [
          {interpMode=RCIM_Cubic, tangentMode=RCTM_User, time=0, value=0.9}
          {interpMode=RCIM_Cubic, tangentMode=RCTM_User, time=0.5, value=1}
          {tangentMode=RCTM_User, time=1, value=0.9}
         ]
        }}}
      CurveIndex=&Frac_Float(Float=&Add_Float(A=&RandomRangeFloat(Minimum=0, Maximum=1), B=$Particles.NormalizedAge))
     )
    )) {isSetParameters=true}
   &DynamicMaterialParameters(
    Index 0 Param 1=1
    Index 0 Param 2=1
    Index 0 Param 4=1
    Module.Index0Float0Write=true
    Module.Index0Float1Write=true
    Module.Index0Float2Write=true
    Module.Index0Float3Write=true
   ) {enabled=false}
   &ScaleColor(
    Linear Color Curve=DataInterface() {propertyValues {
      AlphaCurve {keys [<at 0 1 RCIM_Cubic>]}
      BlueCurve {keys [<at 0 0.2313 RCIM_Cubic>,<at 0.5 0.6487 RCIM_Cubic>,<at 1 0.2313 RCIM_Cubic>]}
      GreenCurve {keys [<at 0 0.4078 RCIM_Cubic>,<at 0.5 0.7293 RCIM_Cubic>,<at 1 0.4078 RCIM_Cubic>]}
      RedCurve {keys [<at 0 1 RCIM_Cubic>,<at 0.5 1 RCIM_Cubic>,<at 1 1 RCIM_Cubic>]}
     }}
    Scale Mode=2
   )
  ]
  rendererClasses [&NiagaraMeshRendererProperties]
  renderers [{rendererClass=&NiagaraMeshRendererProperties, rendererIndex=0}]
 }
 SmallCubesOut {
  bEnabled=false
  emitterUpdateScript [
   &EmitterState()
   &SpawnRate(SpawnRate=4, Use Spawn Probability=false)
   &SpawnBurst_Instantaneous(Spawn Count=5, Use Spawn Probability=false, UseLoopCountLimit=false) {enabled=false}
  ]
  particleSpawnScript [
   &InitializeParticle(
    Color Mode=::NewEnumerator0
    Lifetime Max=1.5
    Lifetime Mode=::NewEnumerator1
    Mesh Scale Mode=::NewEnumerator1
    Mesh Uniform Scale Max=0.01
    Mesh Uniform Scale Min=0.008
    UsePositionOffset=false
    Write Material Random=true
   )
   &InitialMeshOrientation(Use Rotation Vector=false)
   &ShapeLocation(
    Cylinder Height=2
    Cylinder Radius=5
    MaskSpawnGroup=false
    Override Seed=false
    Rotation Mode=::NewEnumerator1
    Shape Primitive=::NewEnumerator1
    Surface Only=false
    UseOwnerScale=true
   )
   &AddVelocity(
    Module.Use Falloff Distance=false
    Override Seed=false
    Velocity Mode=::NewEnumerator1
    Velocity Speed=&RandomRangeFloat(Minimum=9, Maximum=12)
   )
   &CameraOffset(Camera Offset Amount=-10)
  ]
  particleUpdateScript [
   &ParticleState()
   &ScaleMeshSize(Scale Factor=&VectorFromFloat(Value=&FloatFromCurve(FloatCurve=DataInterface() {propertyValues {Curve {
         defaultValue=1
         keys [
          {arriveTangent=6.6152, interpMode=RCIM_Cubic, leaveTangent=6.6152, tangentMode=RCTM_User, time=0, value=0}
          {arriveTangent=3.1761, interpMode=RCIM_Cubic, leaveTangent=3.1761, time=0.1033, value=0.9024}
          {
           arriveTangent=0.0339
           arriveTangentWeight=0.5121
           interpMode=RCIM_Cubic
           leaveTangent=0.0339
           leaveTangentWeight=0.5121
           tangentMode=RCTM_User
           time=0.3149
           value=1
          }
          {arriveTangent=-3.0036, interpMode=RCIM_Cubic, leaveTangent=-3.0036, tangentMode=RCTM_User, time=1, value=0}
         ]
        }}})))
   &UpdateMeshOrientation(Rotation Rate=&RandomRangeFloat(Minimum=0.1, Maximum=0.3), Rotation Vector=<1 1 1>)
   &Drag(
    Drag=&FloatFromCurve(
     FloatCurve=DataInterface() {propertyValues {Curve {
        defaultValue=1
        keys [
         {interpMode=RCIM_Cubic, tangentMode=RCTM_User, time=0, value=0}
         {arriveTangent=5.6533, interpMode=RCIM_Cubic, leaveTangent=5.6533, tangentMode=RCTM_User, time=0.223, value=1}
        ]
       }}}
     Scale Curve=&RandomRangeFloat(Minimum=0.5, Maximum=1)
    )
    Use Linear Drag=true
    Use Rotational Drag=true
   )
   &SolveForcesAndVelocity(Clamp Velocity=false, Limit Acceleration=false)
   &ScaleColor(
    Linear Color Curve=DataInterface() {propertyValues {
      AlphaCurve {keys [<at 0 1 RCIM_Cubic>]}
      BlueCurve {keys [<at 0.5 0.6487 RCIM_Cubic>,<at 1 0.2313 RCIM_Cubic>]}
      GreenCurve {keys [<at 0.5 0.7293 RCIM_Cubic>,<at 1 0.4078 RCIM_Cubic>]}
      RedCurve {keys [<at 0.5 1 RCIM_Cubic>,<at 1 1 RCIM_Cubic>]}
     }}
    Scale Mode=2
   )
  ]
  rendererClasses [&NiagaraMeshRendererProperties]
  renderers [{rendererClass=&NiagaraMeshRendererProperties, rendererIndex=0}]
 }
}
summary {
 systemName=NS_Guitar_VoyagerCuboid_Idle
 emitters [
  {emitterName=MainCube, rendererClasses [&NiagaraMeshRendererProperties, &NiagaraLightRendererProperties]}
  {bEnabled=false, emitterName=SmallCubesOut, rendererClasses [&NiagaraMeshRendererProperties]}
  {emitterName=Hunt_Lightning, rendererClasses [&NiagaraMeshRendererProperties]}
  {emitterName=SmallCubesLinger, rendererClasses [&NiagaraMeshRendererProperties]}
  {emitterName=LightningStrings, rendererClasses [&NiagaraMeshRendererProperties]}
 ]
 userVariables [
  User.bIsFrontEnd {type=NiagaraBool}
  User.bIsFrontEndPreview {type=NiagaraBool}
  User.Variant {type=NiagaraFloat, defaultValue=0, description=""}
 ]
}
