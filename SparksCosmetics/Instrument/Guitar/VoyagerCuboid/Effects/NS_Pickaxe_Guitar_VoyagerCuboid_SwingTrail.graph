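# Niagara swing trail effect for the VoyagerCuboid guitar pickaxe, built from three emitters.
# Trail_New and Lightning draw ribbons through shared anim-trail functions, with skeletal mesh and ribbon ID inputs exposed, fading by particle and emitter age.
# Both ribbon emitters tint by time of day, while Lightning adds jittered positions and animated glow colors.
# IdleCubes spawns short-lived, tumbling, shrinking mesh and sprite particles in a sphere that inherit velocity.
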NiagaraSystem /SparksCosmetics/Instrument/Guitar/VoyagerCuboid/Effects/NS_Pickaxe_Guitar_VoyagerCuboid_SwingTrail
dependencies {
 usedDataInterfaces [
  &NiagaraDataInterfaceCurve
  &NiagaraDataInterfaceColorCurve
  &NiagaraDataInterfaceMeshRendererInfo
  &NiagaraDataInterfaceSpriteRendererInfo
  &NiagaraDataInterfaceArrayFloat
  &NiagaraDataInterfaceSocketReader
 ]
 usedDynamicInputs [&SelectFloatFromArray]
 usedModules [
  &self:SystemSpawnScript.SystemScriptSource.SystemScriptGraph.NiagaraNodeAssignment_2.SetVariables_#1
  &self:SystemSpawnScript.SystemScriptSource.SystemScriptGraph.NiagaraNodeAssignment_3.SetVariables_#2
  &self:SystemSpawnScript.SystemScriptSource.SystemScriptGraph.NiagaraNodeAssignment_4.SetVariables_#3
  &SystemState
 ]
}
emitters {
 IdleCubes {
  emitterUpdateScript [
   &EmitterState(Life Cycle Mode=::NewEnumerator1, Loop Behavior=::NewEnumerator1, Loop Duration=0.25, UseLoopDelay=false)
   &SpawnRate(SpawnRate=40, Use Spawn Probability=true)
  ]
  particleSpawnScript [
   &InitializeParticle(
    Color Maximum=<rgba 0.75 0.6148 0.622 1>
    Color Minimum=<rgba 0.45 0.3797 0.3447 1>
    Color Mode=::NewEnumerator2
    Lifetime=&RandomRangeFloat(Minimum=0.1, Maximum=0.25)
    Mesh Scale Mode=::NewEnumerator3
    Mesh Uniform Scale=&RandomRangeFloat(Minimum=0.4, Maximum=0.75)
    Position Offset=&RandomRangeVector(Minimum=<500 0 0>, Maximum=<-500 0 0>)
    UsePositionOffset=false
    Write Material Random=true
   )
   &SocketLocation(
    SamplePosOffset=true
    Socket Index=&RandomRangeInt(Minimum=1, Maximum=0)
    Socket Reader=$System.Socket Reader
    Socket Sampling Mode=::NewEnumerator3
    VelocitySpeedLimit=false
   ) {enabled=false}
   &ShapeLocation(MaskSpawnGroup=false, Override Seed=false, Sphere Radius=15, UseOwnerScale=true)
   &ShapeLocation(Box Size=<100 5 50>, MaskSpawnGroup=false, Override Seed=false, Shape Primitive=::NewEnumerator2, Surface Only=false, UseOwnerScale=true) {enabled=false, module=ShapeLocation001}
   &SubUVAnimation(
    Mesh Renderer=DataInterface() {propertyValues {MeshRenderer=&self:IdleCubes_0.NiagaraMeshRendererProperties_0}}
    Output Alternate Renderer Binding=true
    Random Start Frame=true
    Renderer Type:ENiagaraMeshOrSprite=::NewEnumerator1
    SubUV Lookup Index Scale=0
    UseEndFrame=false
    UseStartFrame=false
   )
   &AddVelocity(
    Module.Use Falloff Distance=false
    Override Seed=false
    Velocity Mode=::NewEnumerator1
    Velocity Speed=&RandomRangeFloat(Minimum=25, Maximum=10)
   )
   &ApplyInitialForces()
  ]
  particleUpdateScript [
   &ParticleState()
   &UpdateMeshOrientation(
    Rotation Rate=&RandomRangeFloat(Minimum=5, Maximum=2)
    Rotation Vector=&RandomRangeVector(Minimum=<-1 -1 -1>, Maximum=<1 1 1>)
   )
   &InheritVelocity(
    Inherited Velocity Speed Limit:NiagaraFloat=&FloatFromCurve(FloatCurve=DataInterface() {propertyValues {Curve {keys [<at 0 1>,<at 0.4886 0>]}}}, Scale Curve=100)
    Speed Limit=true
    UseSpeedThreshold=false
   )
   &ScaleMeshSize(Scale Factor=&Multiply_VectorByFloat(
     Vector=<1 1 1>
     Float=&FloatFromCurve(FloatCurve=DataInterface() {propertyValues {Curve {keys [{interpMode=RCIM_Cubic, tangentMode=RCTM_User, time=0, value=1}, {arriveTangent=-3.6168, tangentMode=RCTM_User, time=1, value=0}]}}})
    )) {module=ScaleMeshSize001}
   &DynamicMaterialParameters(
    Index 0 Param 1=&FloatFromCurve(FloatCurve=DataInterface() {propertyValues {Curve {keys [<at 0 0>,<at 1 4>]}}})
    Index 0 Param 2=1
    Index 0 Param 3=&FloatFromCurve(
     FloatCurve=DataInterface() {propertyValues {Curve {keys [{interpMode=RCIM_Cubic, tangentMode=RCTM_User, time=0, value=2}, {arriveTangent=-2.5233, tangentMode=RCTM_User, time=1.0059, value=1}]}}}
     Scale Curve=&RandomRangeFloat(Minimum=0.75, Maximum=1)
    )
    Module.Index0Float0Write=true
    Module.Index0Float1Write=true
    Module.Index0Float2Write=true
    Module.Index0Float3Write=true
   )
   &ScaleColor(
    Scale RGB=&Multiply_VectorByFloat(
     Vector=<1 1 1>
     Float=&FloatFromCurve(FloatCurve=DataInterface() {propertyValues {Curve {keys [{interpMode=RCIM_Cubic, tangentMode=RCTM_User, time=0, value=2}, {arriveTangent=-1.9127, tangentMode=RCTM_User, time=1, value=1}]}}})
    )
    ScaleA=true
    ScaleRGB=true
   )
   &SolveForcesAndVelocity(Clamp Velocity=false, Limit Acceleration=false)
  ]
  rendererClasses [&NiagaraSpriteRendererProperties, &NiagaraMeshRendererProperties]
  renderers [
   {rendererClass=&NiagaraSpriteRendererProperties}
   {rendererClass=&NiagaraMeshRendererProperties, rendererIndex=1}
  ]
 }
 Lightning {
  emitterSpawnScript [&FN_RibbonIDOffsetUtility()]
  emitterUpdateScript [
   &EmitterState(Life Cycle Mode=::NewEnumerator1, Loop Behavior=::NewEnumerator1, Loop Delay=0.05, Loop Duration=0.25, UseLoopDelay=true)
   &FN_SpawnAnimTrail(Number of Anim Trails=2)
  ]
  particleSpawnScript [
   &FN_InitializeAnimTrailBoneIndecies()
   &FN_InitializeAnimTrailBoneIndecies(AnimTrail Selection=1) {module=FN_InitializeAnimTrailBoneIndecies001}
   &FN_InitializeAnimTrail(
    Lifetime=&MaskFloat(Mask Source=0, Mask Parameter=?, Mask=0, Masked Float=0.2, Passthrough Float=0.2)
    Position Offset=&TransformVector(Vector=&RandomRangeVector(Minimum=<-10 0 0>, Maximum=<10 0 0>), Destination Space=::World)
   )
  ]
  particleUpdateScript [
   &FN_UpdateAnimTrail() {module=FN_UpdateAnimTrail001}
   &ScaleColor(
    Scale Alpha=&FloatFromCurve(
     FloatCurve=DataInterface() {propertyValues {Curve {keys [
         {interpMode=RCIM_Cubic, tangentMode=RCTM_User, time=0, value=1}
         {arriveTangent=-2.8555, interpMode=RCIM_Cubic, leaveTangent=-2.8555, tangentMode=RCTM_User, time=1, value=0}
        ]}}}
     Scale Curve=&FloatFromCurve(
      FloatCurve=DataInterface() {propertyValues {Curve {keys [
          {arriveTangent=27.7929, interpMode=RCIM_Cubic, leaveTangent=27.7929, tangentMode=RCTM_User, time=0, value=0}
          {arriveTangent=-0.0759, interpMode=RCIM_Cubic, leaveTangent=-0.0759, tangentMode=RCTM_User, time=0.1005, value=1}
          {arriveTangent=-2.4402, interpMode=RCIM_Cubic, leaveTangent=-2.4402, tangentMode=RCTM_User, time=1.4725, value=0}
         ]}}}
      CurveIndex=$Emitter.Age
     )
    )
    Scale RGB=&MakeVectorFromLinearColorRGB(Color=&FNTimeOfDayColor(Night Color=<rgba 0.3427 0.3074 1 1>, Morning Color=<rgba 0.6 0.6262 0.8 1>, Day Color=<rgba 1 1 1 1>, Evening Color=<rgba 0.6 0.6 0.6 1>))
    ScaleA=true
    ScaleRGB=true
   )
   /SetVariables(
    Particles.GlowColor=&MultiplyLinearColor(
     Linear Color=$Particles.Color
     Scale Factor=&MakeVector4_FromLinearColor(LinearColor=&ColorFromCurve(ColorCurve=DataInterface() {propertyValues {
         AlphaCurve {keys [<at 0.05 0>,<at 0.1929 1 RCIM_Cubic>,<at 0.473 0.3>,<at 0.8599 0.0798>,<at 1 0 RCIM_Cubic>]}
         BlueCurve {keys [<at 0.0766 0>,<at 1 0 RCIM_Cubic>]}
         GreenCurve {keys [<at 0.0766 0.4737>,<at 1 0.1425 RCIM_Cubic>]}
         RedCurve {keys [<at 0.0766 1>,<at 1 0.5 RCIM_Cubic>]}
        }}))
    )
    Particles.LightningColor:LinearColor=&MultiplyLinearColor(
     Linear Color=$Particles.Color
     Scale Factor=&MakeVector4_FromLinearColor(LinearColor=&MaskLinearColor(
       Mask Source=0
       Mask Parameter=?
       Mask=0
       Masked Linear Color=&ColorFromCurve(ColorCurve=DataInterface() {propertyValues {
          AlphaCurve {keys [<at 0.0451 0>, {arriveTangent=0.704, interpMode=RCIM_Cubic, time=0.1791, value=1}, <at 0.4806 0.3>, <at 0.7616 0.0836>, <at 1 0 RCIM_Cubic>]}
          BlueCurve {keys [<at 0.0766 0>,<at 0.5082 0>,<at 1 0 RCIM_Cubic>]}
          GreenCurve {keys [<at 0.0766 1.7293>,<at 0.5082 1.1538>,<at 1 0.231 RCIM_Cubic>]}
          RedCurve {keys [<at 0.0766 3>,<at 0.5082 4>,<at 1 1 RCIM_Cubic>]}
         }})
       Passthrough Linear Color=&ColorFromCurve(ColorCurve=DataInterface() {propertyValues {
          AlphaCurve {keys [
            <at 0.042 0>
            {arriveTangent=1.8414, interpMode=RCIM_Cubic, time=0.1241, value=0.5}
            <at 0.2174 0.3>
            <at 0.3266 0.0836>
            <at 0.4405 0 RCIM_Cubic>
           ]}
          BlueCurve {keys [<at 0.0766 0>,<at 0.5082 0>,<at 1 0 RCIM_Cubic>]}
          GreenCurve {keys [<at 0.0766 1.7293>,<at 0.5082 1.1538>,<at 1 0.231 RCIM_Cubic>]}
          RedCurve {keys [<at 0.0766 3>,<at 0.5082 4>,<at 1 1 RCIM_Cubic>]}
         }})
      ))
    )
   ) {isSetParameters=true}
   &ScaleRibbonWidth(Ribbon Width Scale=&FloatFromCurve(FloatCurve=DataInterface() {propertyValues {Curve {
        defaultValue=1
        keys [
         {arriveTangent=0.829, interpMode=RCIM_Cubic, leaveTangent=0.829, tangentMode=RCTM_User, time=0, value=1}
         {interpMode=RCIM_Cubic, tangentMode=RCTM_User, time=1, value=1.5}
        ]
       }}}))
  ]
  rendererClasses [&NiagaraRibbonRendererProperties]
  renderers [
   {rendererClass=&NiagaraRibbonRendererProperties}
   {rendererClass=&NiagaraRibbonRendererProperties, rendererIndex=1}
  ]
 }
 Trail_New {
  emitterSpawnScript [&FN_RibbonIDOffsetUtility()]
  emitterUpdateScript [
   &EmitterState()
   &FN_SpawnAnimTrail()
  ]
  particleSpawnScript [
   &FN_InitializeAnimTrailBoneIndecies()
   &FN_InitializeAnimTrailBoneIndecies(AnimTrail Selection=1) {enabled=false, module=FN_InitializeAnimTrailBoneIndecies001}
   &FN_InitializeAnimTrail(AnimTrailWidth=0.9, Lifetime=0.15)
  ]
  particleUpdateScript [
   &FN_UpdateAnimTrail() {module=FN_UpdateAnimTrail001}
   &ScaleColor(
    Scale Alpha=&FloatFromCurve(
     FloatCurve=DataInterface() {propertyValues {Curve {keys [{interpMode=RCIM_Cubic, tangentMode=RCTM_User, time=0, value=1}, {interpMode=RCIM_Cubic, tangentMode=RCTM_User, time=1, value=0}]}}}
     Scale Curve=&FloatFromCurve(
      FloatCurve=DataInterface() {propertyValues {Curve {keys [
          {interpMode=RCIM_Cubic, tangentMode=RCTM_User, time=0.6557, value=1}
          {arriveTangent=0.0144, interpMode=RCIM_Cubic, leaveTangent=0.0144, tangentMode=RCTM_User, time=0.893, value=0}
         ]}}}
      CurveIndex=$Emitter.Age
     )
    )
    Scale RGB=&MakeVectorFromLinearColorRGB(Color=&FNTimeOfDayColor(
      Night Color=<rgba 0.4 0.6221 1 1>
      Morning Color=<rgba 0.6375 0.6654 0.85 1>
      Day Color=<rgba 1 1 1 1>
      Evening Color=<rgba 0.65 0.65 0.65 1>
     ))
    ScaleA=true
    ScaleRGB=true
   )
  ]
  rendererClasses [&NiagaraRibbonRendererProperties]
  renderers [
   {rendererClass=&NiagaraRibbonRendererProperties}
   {rendererClass=&NiagaraRibbonRendererProperties, rendererIndex=1}
   {rendererClass=&NiagaraRibbonRendererProperties, rendererIndex=2}
  ]
 }
}
summary {
 systemName=NS_Pickaxe_Guitar_VoyagerCuboid_SwingTrail
 emitters [
  {emitterName=Lightning, rendererClasses [&NiagaraRibbonRendererProperties]}
  {emitterName=IdleCubes, rendererClasses [&NiagaraSpriteRendererProperties, &NiagaraMeshRendererProperties]}
  {emitterName=Trail_New, rendererClasses [&NiagaraRibbonRendererProperties]}
 ]
 userVariables [
  User.RibbonIDOffset {type=NiagaraInt32}
  User.SkeletalMesh {
   type=NiagaraDataInterfaceSkeletalMesh
   defaultValue {
    struct=NiagaraExt_VariableValue_DataInterface
    value {dataInterface=&self:NiagaraDataInterfaceSkeletalMesh_0}
   }
  }
  User.SwingSectionIndex {type=NiagaraInt32}
 ]
}
