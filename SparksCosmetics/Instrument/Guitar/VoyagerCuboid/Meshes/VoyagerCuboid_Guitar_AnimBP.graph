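# Animation Blueprint for the VoyagerCuboid guitar cosmetic, procedurally posing its floating stone shards and middle cube.
# Each update, the EventGraph derives phase-shifted sine translations for small and large shards and steadily growing cube rotations from real time.
# The AnimGraph adjusts the offset and middle-cube bones on the input pose, caches it, then chains additive bone-space Modify Bone nodes over the dyn_stone bones, blended at partial alpha.
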AnimBlueprint /SparksCosmetics/Instrument/Guitar/VoyagerCuboid/Meshes/VoyagerCuboid_Guitar_AnimBP
target_skeleton=&/VoyagerCuboid_Guitar_Skeleton
graphs [
 AnimGraph {nodes [
   n1 = Local To Component(LocalPose=Use cached pose 'CRCache_CubeRotation_1'())
   n2 = "Transform (Modify) Bone - Bone: dyn_stone_11"!(
    ComponentPose="Transform (Modify) Bone - Bone: dyn_stone_11"(
     ComponentPose="Transform (Modify) Bone - Bone: dyn_stone_12"!(ComponentPose=@n1, Translation="0,0,0", Rotation="0,0,0", Scale="1,1,1") -> (Pose) {alpha=0.75, translationMode=BMM_Additive, translationSpace=BCS_BoneSpace, boneToModify {boneName=dyn_stone_12}}
     Translation="0,0,0"
     Rotation="0,0,0"
     Scale="1,1,1"
    ) {alpha=0.75, translationMode=BMM_Additive, translationSpace=BCS_BoneSpace, boneToModify {boneName=dyn_stone_11}}
    Translation="0,0,1.6"
    Rotation="0,0,0"
    Scale="1,1,1"
   ) -> (Pose) {
    alpha=0.75
    rotationMode=BMM_Additive
    rotationSpace=BCS_BoneSpace
    translationMode=BMM_Additive
    translationSpace=BCS_BoneSpace
    boneToModify {boneName=dyn_stone_11}
    translation {z=1.6}
   }
   n3 = "Transform (Modify) Bone - Bone: dyn_stone_09"!(
    ComponentPose="Transform (Modify) Bone - Bone: dyn_stone_10"!(ComponentPose=@n2, Translation="0,0,0", Rotation="0,0,0", Scale="1,1,1") -> (Pose) {alpha=0.75, translationMode=BMM_Additive, translationSpace=BCS_BoneSpace, boneToModify {boneName=dyn_stone_10}}
    Translation="0,0,0"
    Rotation="0,0,0"
    Scale="1,1,1"
   ) -> (Pose) {alpha=0.75, translationMode=BMM_Additive, translationSpace=BCS_BoneSpace, boneToModify {boneName=dyn_stone_09}}
   n4 = "Transform (Modify) Bone - Bone: dyn_stone_07"!(
    ComponentPose="Transform (Modify) Bone - Bone: dyn_stone_07"(
     ComponentPose="Transform (Modify) Bone - Bone: dyn_stone_08"!(ComponentPose=@n3, Translation="0,0,0", Rotation="0,0,0", Scale="1,1,1") -> (Pose) {alpha=0.75, translationMode=BMM_Additive, translationSpace=BCS_BoneSpace, boneToModify {boneName=dyn_stone_08}}
     Translation="0,0,0"
     Rotation="0,0,0"
     Scale="1,1,1"
    ) {alpha=0.75, translationMode=BMM_Additive, translationSpace=BCS_BoneSpace, boneToModify {boneName=dyn_stone_07}}
    Translation="0,-0.8,0"
    Rotation="0,0,0"
    Scale="1,1,1"
   ) -> (Pose) {alpha=0.75, translationMode=BMM_Additive, translationSpace=BCS_BoneSpace, boneToModify {boneName=dyn_stone_07}, translation {y=-0.8}}
   n5 = "Transform (Modify) Bone - Bone: dyn_stone_05"!(
    ComponentPose="Transform (Modify) Bone - Bone: dyn_stone_06"!(ComponentPose=@n4, Translation="0,0,0", Rotation="0,0,0", Scale="1,1,1") -> (Pose) {alpha=0.75, translationMode=BMM_Additive, translationSpace=BCS_BoneSpace, boneToModify {boneName=dyn_stone_06}}
    Translation="0,0,0"
    Rotation="0,0,0"
    Scale="1,1,1"
   ) -> (Pose) {alpha=0.75, translationMode=BMM_Additive, translationSpace=BCS_BoneSpace, boneToModify {boneName=dyn_stone_05}}
   n6 = "Transform (Modify) Bone - Bone: dyn_stone_02"!(
    ComponentPose="Transform (Modify) Bone - Bone: dyn_stone_03"!(
     ComponentPose="Transform (Modify) Bone - Bone: dyn_stone_04"!(ComponentPose=@n5, Translation="0,0,0", Rotation="0,0,0", Scale="1,1,1") -> (Pose) {alpha=0.75, translationMode=BMM_Additive, translationSpace=BCS_BoneSpace, boneToModify {boneName=dyn_stone_04}}
     Translation="0,0,0"
     Rotation="0,0,0"
     Scale="1,1,1"
    ) -> (Pose) {alpha=0.75, translationMode=BMM_Additive, translationSpace=BCS_BoneSpace, boneToModify {boneName=dyn_stone_03}}
    Translation="0,0,0"
    Rotation="0,0,0"
    Scale="1,1,1"
   ) -> (Pose) {alpha=0.75, translationMode=BMM_Additive, translationSpace=BCS_BoneSpace, boneToModify {boneName=dyn_stone_02}}
   Output Pose(Result=Component To Local(ComponentPose="Transform (Modify) Bone - Bone: dyn_stone_01"!(ComponentPose=@n6, Translation="0,0,0", Rotation="0,0,0", Scale="1,1,1") -> (Pose) {alpha=0.75, translationMode=BMM_Additive, translationSpace=BCS_BoneSpace, boneToModify {boneName=dyn_stone_01}}))
   n7 = "Transform (Modify) Bone - Bone: offset"!(
    ComponentPose=Local To Component(LocalPose=Input Pose())
    bAlphaBoolEnabled:bool=True
    Translation="2.5,-2,4"
    Rotation="0,0,2"
    Scale="1,1,1"
   ) -> (Pose) {
    alphaInputType=Bool
    rotationMode=BMM_Additive
    rotationSpace=BCS_BoneSpace
    scaleMode=BMM_Additive
    scaleSpace=BCS_BoneSpace
    translationMode=BMM_Additive
    translationSpace=BCS_BoneSpace
    alphaBoolBlend {blendOption=Linear}
    boneToModify {boneName=offset}
    rotation {roll=2}
    translation=<2.5 -2 4>
   }
   n8 = "Transform (Modify) Bone - Bone: dyn_midcube_01"(ComponentPose=@n7, Alpha=1, Translation="0,0,0", Rotation="0,0,0", Scale="1,1,1") {rotationMode=BMM_Additive, rotationSpace=BCS_BoneSpace, translationSpace=BCS_BoneSpace, boneToModify {boneName=dyn_midcube_01}}
   Save cached pose 'CRCache_CubeRotation_1'(Pose=Component To Local(ComponentPose=@n8)) {cacheName=CRCache_CubeRotation_1}
  ]}
 EventGraph {nodes [
   Event BlueprintUpdateAnimation()
   > n1 = Sequence()
   > n2 = Set Range(Range:double=0.8) -> (Output_Get:double)
   n3 = float - float(A=GetRealTimeSeconds(), B=0.6)
   n4 = float / float(A="Sin (Radians)"(A=float * float(A=@n3, B=2)), B=0.5) {is=title K2Node_CallFunction}
   > "Set Translate Shard 1 "("Translate Shard 1 ":Vector=MakeVector(X=Remap Float(Input=@n4, OldRange=(X=-2,Y=2), NewRange=(X=0,Y=3)) -> (Output=~), Y=0)) -> (Output_Get:Vector)
   n5 = float - float(A=GetRealTimeSeconds(), B=0.2)
   n6 = float / float(A="Sin (Radians)"(A=float * float(A=@n5, B=1.5)), B=@n2)
   > Set Translate Shard 2(Translate Shard 2:Vector=MakeVector(X=Remap Float(Input=@n6, OldRange=(X=-1.2,Y=1.2), NewRange=(X=0.5,Y=2.5)) -> (Output=~), Y=0)) -> (Output_Get:Vector)
   n7 = GetRealTimeSeconds()
   n8 = "Sin (Radians)"(A=float * float(A=float - float(A=@n7, B=0.4), B=1.5))
   > Set Translate Shard 3(Translate Shard 3:Vector=MakeVector(X=float / float(A=@n8, B=@n2), Y=0)) -> (Output_Get:Vector)
   n9 = GetRealTimeSeconds()
   n10 = "Sin (Radians)"(A=float * float(A=float - float(A=@n9, B=0.6), B=1.5))
   > Set Translate Shard 4(Translate Shard 4:Vector=MakeVector(X=float / float(A=@n10, B=@n2), Y=0)) -> (Output_Get:Vector)
   n11 = GetRealTimeSeconds()
   n12 = "Sin (Radians)"(A=float * float(A=float - float(A=@n11, B=0.8), B=1.5))
   > Set Translate Shard 5(Translate Shard 5:Vector=MakeVector(X=float / float(A=@n12, B=@n2), Y=0)) -> (Output_Get:Vector)
   n13 = GetRealTimeSeconds()
   n14 = "Sin (Radians)"(A=float * float(A=float - float(A=@n13, B=1), B=1.5))
   > Set Translate Shard 6(Translate Shard 6:Vector=MakeVector(X=float / float(A=@n14, B=1), Y=0)) -> (Output_Get:Vector)
   n15 = GetRealTimeSeconds()
   n16 = "Sin (Radians)"(A=float * float(A=float - float(A=@n15, B=1), B=3.5))
   > Set Translate Shard Large 1(Translate Shard Large 1:Vector=MakeVector(X=float / float(A=@n16, B=@n2), Y=0)) -> (Output_Get:Vector)
   n17 = GetRealTimeSeconds()
   n18 = "Sin (Radians)"(A=float * float(A=float - float(A=@n17, B=0.75), B=3.5))
   > Set Translate Shard Large 2(Translate Shard Large 2:Vector=MakeVector(X=float / float(A=@n18, B=@n2), Y=0)) -> (Output_Get:Vector)
   n19 = GetRealTimeSeconds()
   n20 = "Sin (Radians)"(A=float * float(A=float - float(A=@n19, B=0.5), B=3.5))
   > Set Translate Shard Large 3(Translate Shard Large 3:Vector=MakeVector(X=float / float(A=@n20, B=@n2), Y=0)) -> (Output_Get:Vector)
   n21 = float * float(A=float - float(A=GetRealTimeSeconds(), B=1), B=-10)
   Set Rotate Shard Cube(execute=@n1.then_1, Rotate Shard Cube:Rotator=MakeRotator(Yaw=float * float(A=@n21, B=1.5) {is=title K2Node_PromotableOperator})) -> (Output_Get:Rotator)
   n22 = GetRealTimeSeconds()
   n23 = float * float(A=float * float(A=float - float(A=@n22, B=1), B=-5) {is=title K2Node_CommutativeAssociativeBinaryOperator}, B=0.5)
   > Set Rotate Shard Cube_2(Rotate Shard Cube_2:Rotator=MakeRotator(Roll=@n23, Pitch=@n23, Yaw=@n23)) -> (Output_Get:Rotator)
   n24 = GetRealTimeSeconds()
   n25 = float * float(A=float * float(A=float - float(A=@n24, B=1), B=10) {is=title K2Node_CommutativeAssociativeBinaryOperator}, B=1) {is=title K2Node_PromotableOperator}
   > Set Rotate Shard Cube_3(Rotate Shard Cube_3:Rotator=MakeRotator(Roll=@n25, Pitch=@n25, Yaw=@n25)) -> (Output_Get:Rotator)
   n26 = GetRealTimeSeconds()
   n27 = float * float(A=float * float(A=float - float(A=@n26, B=1), B=-20) {is=title K2Node_CommutativeAssociativeBinaryOperator}, B=1.3) {is=title K2Node_PromotableOperator}
   > Set Rotate Shard Cube_4(Rotate Shard Cube_4:Rotator=MakeRotator(Roll=@n27, Pitch=@n27, Yaw=@n27)) -> (Output_Get:Rotator)
   TryGetPawnOwner()
   TryGetPawnOwner()
  ]}
]
variables [
 Range {type=number}
 Rotate Shard Cube {type=Rotator}
 Rotate Shard Cube_2 {type=Rotator}
 Rotate Shard Cube_3 {type=Rotator}
 Rotate Shard Cube_4 {type=Rotator}
 "Translate Shard 1 " {type=Vector}
 Translate Shard 2 {type=Vector}
 Translate Shard 3 {type=Vector}
 Translate Shard 4 {type=Vector}
 Translate Shard 5 {type=Vector}
 Translate Shard 6 {type=Vector}
 Translate Shard Large 1 {type=Vector}
 Translate Shard Large 2 {type=Vector}
 Translate Shard Large 3 {type=Vector}
]
